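# MetaSound Source preset for the Chiller Snowball Launcher's third-person fire sound, built on MSS_Base_Weapon_Fire_3p_Simple.
# It supplies close, medium and distant sets of three fire variations, with per-layer gains and shared ranged-weapon distance curves.
# Inputs set random pitch, volume and EQ ranges, fire timing and a friendly-fire volume for the parent's logic.
# Stereo output uses gunfire concurrency, an owner-limited stop-oldest override and third-person fire volume modulation.
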MetaSoundSource /WeaponsUnvaulted/Assets/Audio/Chiller_Launcher/MSS_Chiller_Snowball_Launcher_3p_Preset
preset_parent=/FortMetaSoundWeapon/Items/Metasounds/MSS_Base_Weapon_Fire_3p_Simple
inputs [
 bHasUnlimitedAmmo {type=Bool}
 BurstRate {type=Time}
 bUseSingleVoice {type=Bool}
 Character.Affiliation {type=Enum:FortAudioAffiliation}
 Character.Health {type=Float}
 Character.UniqueID {type=Int32}
 FireBurst {type=Trigger}
 FireSingleShot {type=Trigger}
 MainFire.DistanceCurves {type=WaveTableBankAsset}
 MainFire.FalloffDistance {type=Float}
 MainFire.SoundsClose {type=WaveAsset:Array}
 MainFire.SoundsDistant {type=WaveAsset:Array}
 MainFire.SoundsMedium {type=WaveAsset:Array}
 MainFire.StopTriggerDelay {type=Time}
 Mixing.CloseGain {type=Float}
 Mixing.DistantGain {type=Float}
 Mixing.Friendly.FriendlyFireWeaponVolume {type=Float}
 Mixing.MediumGain {type=Float}
 Mixing.WeaponOutputVolume {type=Float}
 Random.EQ.A.MaxFreq {type=Float}
 Random.EQ.A.MaxGain {type=Float}
 Random.EQ.A.MinFreq {type=Float}
 Random.EQ.A.MinGain {type=Float}
 Random.MaxPitch {type=Float}
 Random.MaxVolume {type=Float}
 Random.MinPitch {type=Float}
 Random.MinVolume {type=Float}
 RetriggerDelayTime {type=Time}
 RoundsPerBurst {type=Int32}
 StartFiring {type=Trigger}
 StopBurst {type=Trigger}
 StopFiring {type=Trigger}
 UE.Attenuation.Distance {type=Float}
 UE.Source.OnPlay {type=Trigger}
 Weapon.AmmoRemaining {type=Int32}
 Weapon.ClipSize {type=Int32}
 Weapon.FireRate {type=Float}
 Weapon.FireStartTime {type=Float}
 Weapon.IsAimingDownSights {type=Bool}
 Weapon.IsSuppressorEnabled {type=Bool}
]
interfaces [Fort.Character:1.0, Fort.Weapon:1.0, UE.Attenuation:1.0, UE.OutputFormat.Stereo:1.0, UE.Source.OneShot:1.0, UE.Source:1.1]
outputs [
 bIsAlmostFinished {type=Bool}
 "UE.OutputFormat.Stereo.Audio:0" {type=Audio}
 "UE.OutputFormat.Stereo.Audio:1" {type=Audio}
 UE.Source.OneShot.OnFinished {type=Trigger}
]
pages [Default {input_defaults {
   BurstRate=0.4
   Character.Affiliation=0
   Character.Health=1
   Character.UniqueID=0
   FireBurst=null
   FireSingleShot=null
   MainFire.DistanceCurves=&WeaponRanged_DistanceCurves
   MainFire.FalloffDistance=27500
   MainFire.StopTriggerDelay=0.03
   Mixing.CloseGain=1.4
   Mixing.DistantGain=0.6
   Mixing.Friendly.FriendlyFireWeaponVolume=0.7
   Mixing.MediumGain=0.7
   Mixing.WeaponOutputVolume=1
   Random.EQ.A.MaxFreq=20000
   Random.EQ.A.MaxGain=0
   Random.EQ.A.MinFreq=699.9999
   Random.EQ.A.MinGain=-3
   Random.MaxPitch=1.15
   Random.MaxVolume=1.15
   Random.MinPitch=0.95
   Random.MinVolume=0.85
   RetriggerDelayTime=2
   RoundsPerBurst=2
   StartFiring=null
   StopBurst=null
   StopFiring=null
   UE.Attenuation.Distance=25000
   UE.Source.OnPlay=null
   Weapon.AmmoRemaining=30
   Weapon.ClipSize=30
   Weapon.FireRate=0.5
   Weapon.FireStartTime=0
   Weapon.IsAimingDownSights=null
   Weapon.IsSuppressorEnabled=null
   bHasUnlimitedAmmo=null
   bUseSingleVoice="<MetaSoundBuilderResult.SUCCEEDED: 0>"
   MainFire.SoundsClose [
    &/WPN_Chiller_Snowball_Launcher_Fire_01
    &/WPN_Chiller_Snowball_Launcher_Fire_02
    &/WPN_Chiller_Snowball_Launcher_Fire_03
   ]
   MainFire.SoundsDistant [
    &/WPN_Chiller_Snowball_Launcher_Fire_Far_01
    &/WPN_Chiller_Snowball_Launcher_Fire_Far_02
    &/WPN_Chiller_Snowball_Launcher_Fire_Far_03
   ]
   MainFire.SoundsMedium [
    &/WPN_Chiller_Snowball_Launcher_Fire_Mid_01
    &/WPN_Chiller_Snowball_Launcher_Fire_Mid_02
    &/WPN_Chiller_Snowball_Launcher_Fire_Mid_03
   ]
  }}]
properties {
 assetImportData=&self:AssetImportData
 attenuationSettings=&Fort_Weapon_Projectile_NonAtt_Athena_Att
 maxDistance=27500
 numChannels=2
 outputFormat=Stereo
 qualitySetting=Fort_Weapons
 sampleRate=48000
 soundClassObject=&Fort_WeaponRanged_Fire_3P_PSM
 concurrencyOverrides {bLimitToOwner=true, maxCount=2, resolutionRule=StopOldest}
 concurrencySet [&Weapon_Gunfire_Local_Conc, &Weapon_Gunfire_Global_Conc]
 modulationSettings {
  volumeRouting=Union
  volumeModulationDestination {bEnableModulation=true, modulators [/FortMetaSoundWeapon/Items/Modulation/PP_Weapon_Fire_3p_Vol]}
 }
}
